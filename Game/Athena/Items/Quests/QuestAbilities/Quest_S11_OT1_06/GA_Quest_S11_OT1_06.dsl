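; Quest gameplay ability that credits the player for having health and shield both at or above a comparison value at once.
; On activation it checks both starting attributes in case the player spawns qualified, then threshold-watch tasks update a success flag for each.
; TryProgressQuest requires both flags, then rereads the values after a short delay to filter brief spawn-time flashes before sending the MaxHealthMaxShield stat event to the quest manager.
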(event Ability|EventActivateAbility
  (bind (_bsuccessfullyfoundattribute _returnvalue) (Ability|Attribute|GetFloatAttribute (Variables|GAVars|Data|GetOwningPlayerPawn) (Variables|GAVars|Setup|GetAtt_Shield)))
  (bind (_bsuccessfullyfoundattribute_2 _returnvalue_2) (Ability|Attribute|GetFloatAttribute (Variables|GAVars|Data|GetOwningPlayerPawn) (Variables|GAVars|Setup|GetAtt_Health)))
  ;[1 1 1] Check starting health and shield values in case player spawns with threshold success
  (|CallParentFunction)
  (bind _thresholdsuccess (CallFunction|CheckInitialAttributeValue _bsuccessfullyfoundattribute_2 _returnvalue_2))
  (Variables|GAVars|Data|SetHealthSuccess _thresholdsuccess)
  (bind _thresholdsuccess_2 (CallFunction|CheckInitialAttributeValue _bsuccessfullyfoundattribute _returnvalue))
  (Variables|GAVars|Data|SetShieldSuccess _thresholdsuccess_2)
  (CallFunction|TryProgressQuest)
  ;[1 1 1] Wait for attribute changes and check progress again
  (bind (_asynctaskproxy _bmatchescomparison) (Ability|Tasks|WaitForAttributeChangeThreshold (Variables|GAVars|Setup|GetAtt_Health) "GreaterThanOrEqualTo" (Variables|GAVars|Setup|GetComparisonValue))
    (:then
      (bind (_asynctaskproxy_2 _bmatchescomparison_2) (Ability|Tasks|WaitForAttributeChangeThreshold (Variables|GAVars|Setup|GetAtt_Shield) "GreaterThanOrEqualTo" (Variables|GAVars|Setup|GetComparisonValue))
        (:OnChange
          (Variables|GAVars|Data|SetShieldSuccess _bmatchescomparison_2))))
    (:OnChange
      (Variables|GAVars|Data|SetHealthSuccess _bmatchescomparison)))
  (CallFunction|TryProgressQuest))

(event Ability|EventFailedToActivatePassiveAbility
  (params (ActorInfo :type Gameplay%20Ability%20Actor%20Info%20Structure)))

(event Custom|TryProgressQuest
  (bind (_bsuccessfullyfoundattribute _returnvalue) (Ability|Attribute|GetFloatAttribute (Variables|GAVars|Data|GetOwningPlayerPawn) (Variables|GAVars|Setup|GetAtt_Health)))
  (bind (_bsuccessfullyfoundattribute_2 _returnvalue_2) (Ability|Attribute|GetFloatAttribute (Variables|GAVars|Data|GetOwningPlayerPawn) (Variables|GAVars|Setup|GetAtt_Shield)))
  ;[1 1 1] If health and shield meet thresholds, send quest progress
  (if (and (Variables|GAVars|Data|GetHealthSuccess) (Variables|GAVars|Data|GetShieldSuccess))
    ;[1 0.8959 0.3865] Added this delay and additional check to prevent future occurrences of FORT-251834 In that case, the LTM flashed the HP and Shield to 100 briefly when the player jumped from the bus
    (Utilities|FlowControl|Delay 0.05)
    (if (and (>= _returnvalue (Variables|GAVars|Setup|GetComparisonValue)) (>= _returnvalue_2 (Variables|GAVars|Setup|GetComparisonValue)))
      (Quest|SendComplexCustomStatEvent (Variables|GAVars|Data|GetOwningPlayerQuestManager) 0 "(GameplayTags=((TagName=\"Athena.Quests.MaxHealthMaxShield\")))" "(GameplayTags=)" 1))))

(fn CheckInitialAttributeValue
  (params (AttributeFound :type Boolean) (Value :type Float%20%28double-precision%29))
  (if AttributeFound
    _))
